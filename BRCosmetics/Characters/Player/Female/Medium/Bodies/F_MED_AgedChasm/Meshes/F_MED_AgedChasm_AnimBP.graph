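# Post-pose animation blueprint for the F_MED_AgedChasm female cosmetic body, adding secondary motion to its cloth bones.
# It copies the parent mesh pose into a Control Rig, then chains spring-based Anim Dynamics nodes: side bones at half strength, back and front at full.
# Additive bone offsets follow, gated by skydiving (back and side-C bones) and crouching (first side-C bones).
# The event graph has no logic.
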AnimBlueprint /BRCosmetics/Characters/Player/Female/Medium/Bodies/F_MED_AgedChasm/Meshes/F_MED_AgedChasm_AnimBP
target_skeleton=&/F_MED_AgedChasm_Skeleton
graphs [
 AnimGraph {nodes [
   n1 = Control Rig(Source=Copy Pose From Mesh() {bUseAttachedParent=true}) {
    alphaBoolBlend {blendOption=Linear}
    controlRigAssetReference {blueprintRigClass=&/F_MED_AgedChasm_CtrlRig.F_MED_AgedChasm_CtrlRig_C}
   }
   n2 = Anim Dynamics - Bone: dyn_cloth_a_01_l(ComponentPose=Local To Component(LocalPose=@n1), Alpha=0.5, ExternalForce="0,0,0") {
    angularDampingOverride=1
    angularSpringConstant=5000
    bAngularSpring=true
    bOverrideAngularDamping=true
    bUseGravityOverride=true
    gravityScale=0.35
    simulationSpace=RootRelative
    alphaBoolBlend {blendOption=Linear}
    boundBone {boneName=dyn_cloth_a_01_l}
    gravityOverride {z=-980}
    physicsBodyDefinitions [{
      collisionType=OuterSphere
      boundBone {boneName=dyn_cloth_a_01_l}
      boxExtents=<15 25 25>
      constraintSetup {angularLimitsMax=<0 3 6>, angularLimitsMin=<0 -3 -6>, angularTarget=<1 0 0>}
      localJointOffset=<2 0 0>
     }]
   }
   n3 = Anim Dynamics - Bone: dyn_cloth_a_02_l(ComponentPose=@n2, Alpha=0.5, ExternalForce="0,0,0") {
    angularDampingOverride=1
    angularSpringConstant=5000
    bAngularSpring=true
    bOverrideAngularDamping=true
    bUseGravityOverride=true
    gravityScale=0.35
    simulationSpace=RootRelative
    alphaBoolBlend {blendOption=Linear}
    boundBone {boneName=dyn_cloth_a_02_l}
    gravityOverride {z=-980}
    physicsBodyDefinitions [{
      collisionType=OuterSphere
      boundBone {boneName=dyn_cloth_a_02_l}
      boxExtents=<15 25 25>
      constraintSetup {angularLimitsMax=<0 3 6>, angularLimitsMin=<0 -3 -6>, angularTarget=<1 0 0>}
      localJointOffset=<2 0 0>
     }]
   }
   n4 = Anim Dynamics - Bone: dyn_cloth_b_01_l(ComponentPose=@n3, Alpha=0.5, ExternalForce="0,0,0") {
    angularDampingOverride=1
    angularSpringConstant=5000
    bAngularSpring=true
    bOverrideAngularDamping=true
    bUseGravityOverride=true
    gravityScale=0.35
    simulationSpace=RootRelative
    alphaBoolBlend {blendOption=Linear}
    boundBone {boneName=dyn_cloth_b_01_l}
    gravityOverride {z=-980}
    physicsBodyDefinitions [{
      collisionType=OuterSphere
      boundBone {boneName=dyn_cloth_b_01_l}
      boxExtents=<15 25 25>
      constraintSetup {angularLimitsMax=<0 3 6>, angularLimitsMin=<0 -3 -6>, angularTarget=<1 0 0>}
      localJointOffset=<2 0 0>
     }]
   }
   n5 = Anim Dynamics - Bone: dyn_cloth_b_02_l(ComponentPose=@n4, Alpha=0.5, ExternalForce="0,0,0") {
    angularDampingOverride=1
    angularSpringConstant=5000
    bAngularSpring=true
    bOverrideAngularDamping=true
    bUseGravityOverride=true
    gravityScale=0.35
    simulationSpace=RootRelative
    alphaBoolBlend {blendOption=Linear}
    boundBone {boneName=dyn_cloth_b_02_l}
    gravityOverride {z=-980}
    physicsBodyDefinitions [{
      collisionType=OuterSphere
      boundBone {boneName=dyn_cloth_b_02_l}
      boxExtents=<15 25 25>
      constraintSetup {angularLimitsMax=<0 3 6>, angularLimitsMin=<0 -3 -6>, angularTarget=<1 0 0>}
      localJointOffset=<2 0 0>
     }]
   }
   n6 = Anim Dynamics - Bone: dyn_cloth_c_01_l(ComponentPose=@n5, Alpha=0.5, ExternalForce="0,0,0") {
    angularDampingOverride=1
    angularSpringConstant=5000
    bAngularSpring=true
    bOverrideAngularDamping=true
    bUseGravityOverride=true
    gravityScale=0.35
    simulationSpace=RootRelative
    alphaBoolBlend {blendOption=Linear}
    boundBone {boneName=dyn_cloth_c_01_l}
    gravityOverride {z=-980}
    physicsBodyDefinitions [{
      collisionType=OuterSphere
      boundBone {boneName=dyn_cloth_c_01_l}
      boxExtents=<15 25 25>
      constraintSetup {angularLimitsMax=<0 3 6>, angularLimitsMin=<0 -3 -6>, angularTarget=<1 0 0>}
      localJointOffset=<2 0 0>
     }]
   }
   n7 = Anim Dynamics - Bone: dyn_cloth_c_02_l(ComponentPose=@n6, Alpha=0.5, ExternalForce="0,0,0") {
    angularDampingOverride=1
    angularSpringConstant=5000
    bAngularSpring=true
    bOverrideAngularDamping=true
    bUseGravityOverride=true
    gravityScale=0.35
    simulationSpace=RootRelative
    alphaBoolBlend {blendOption=Linear}
    boundBone {boneName=dyn_cloth_c_02_l}
    gravityOverride {z=-980}
    physicsBodyDefinitions [{
      collisionType=OuterSphere
      boundBone {boneName=dyn_cloth_c_02_l}
      boxExtents=<15 25 25>
      constraintSetup {angularLimitsMax=<0 3 6>, angularLimitsMin=<0 -3 -6>, angularTarget=<1 0 0>}
      localJointOffset=<2 0 0>
     }]
   }
   n8 = Anim Dynamics - Bone: dyn_cloth_a_01_r(ComponentPose=@n7, Alpha=0.5, ExternalForce="0,0,0") {
    angularDampingOverride=1
    angularSpringConstant=5000
    bAngularSpring=true
    bOverrideAngularDamping=true
    bUseGravityOverride=true
    gravityScale=0.35
    simulationSpace=RootRelative
    alphaBoolBlend {blendOption=Linear}
    boundBone {boneName=dyn_cloth_a_01_r}
    gravityOverride {z=-980}
    physicsBodyDefinitions [{
      collisionType=OuterSphere
      boundBone {boneName=dyn_cloth_a_01_r}
      boxExtents=<15 25 25>
      constraintSetup {angularLimitsMax=<0 3 6>, angularLimitsMin=<0 -3 -6>, angularTarget=<1 0 0>}
      localJointOffset=<2 0 0>
     }]
   }
   n9 = Anim Dynamics - Bone: dyn_cloth_a_02_r(ComponentPose=@n8, Alpha=0.5, ExternalForce="0,0,0") {
    angularDampingOverride=1
    angularSpringConstant=5000
    bAngularSpring=true
    bOverrideAngularDamping=true
    bUseGravityOverride=true
    gravityScale=0.35
    simulationSpace=RootRelative
    alphaBoolBlend {blendOption=Linear}
    boundBone {boneName=dyn_cloth_a_02_r}
    gravityOverride {z=-980}
    physicsBodyDefinitions [{
      collisionType=OuterSphere
      boundBone {boneName=dyn_cloth_a_02_r}
      boxExtents=<15 25 25>
      constraintSetup {angularLimitsMax=<0 3 6>, angularLimitsMin=<0 -3 -6>, angularTarget=<1 0 0>}
      localJointOffset=<2 0 0>
     }]
   }
   n10 = Anim Dynamics - Bone: dyn_cloth_b_01_r(ComponentPose=@n9, Alpha=0.5, ExternalForce="0,0,0") {
    angularDampingOverride=1
    angularSpringConstant=5000
    bAngularSpring=true
    bOverrideAngularDamping=true
    bUseGravityOverride=true
    gravityScale=0.35
    simulationSpace=RootRelative
    alphaBoolBlend {blendOption=Linear}
    boundBone {boneName=dyn_cloth_b_01_r}
    gravityOverride {z=-980}
    physicsBodyDefinitions [{
      collisionType=OuterSphere
      boundBone {boneName=dyn_cloth_b_01_r}
      boxExtents=<15 25 25>
      constraintSetup {angularLimitsMax=<0 3 6>, angularLimitsMin=<0 -3 -6>, angularTarget=<1 0 0>}
      localJointOffset=<2 0 0>
     }]
   }
   n11 = Anim Dynamics - Bone: dyn_cloth_b_02_r(ComponentPose=@n10, Alpha=0.5, ExternalForce="0,0,0") {
    angularDampingOverride=1
    angularSpringConstant=5000
    bAngularSpring=true
    bOverrideAngularDamping=true
    bUseGravityOverride=true
    gravityScale=0.35
    simulationSpace=RootRelative
    alphaBoolBlend {blendOption=Linear}
    boundBone {boneName=dyn_cloth_b_02_r}
    gravityOverride {z=-980}
    physicsBodyDefinitions [{
      collisionType=OuterSphere
      boundBone {boneName=dyn_cloth_b_02_r}
      boxExtents=<15 25 25>
      constraintSetup {angularLimitsMax=<0 3 6>, angularLimitsMin=<0 -3 -6>, angularTarget=<1 0 0>}
      localJointOffset=<2 0 0>
     }]
   }
   n12 = Anim Dynamics - Bone: dyn_cloth_c_01_r(ComponentPose=@n11, Alpha=0.5, ExternalForce="0,0,0") {
    angularDampingOverride=1
    angularSpringConstant=5000
    bAngularSpring=true
    bOverrideAngularDamping=true
    bUseGravityOverride=true
    gravityScale=0.35
    simulationSpace=RootRelative
    alphaBoolBlend {blendOption=Linear}
    boundBone {boneName=dyn_cloth_c_01_r}
    gravityOverride {z=-980}
    physicsBodyDefinitions [{
      collisionType=OuterSphere
      boundBone {boneName=dyn_cloth_c_01_r}
      boxExtents=<15 25 25>
      constraintSetup {angularLimitsMax=<0 3 6>, angularLimitsMin=<0 -3 -6>, angularTarget=<1 0 0>}
      localJointOffset=<2 0 0>
     }]
   }
   n13 = Anim Dynamics - Bone: dyn_cloth_c_02_r(ComponentPose=@n12, Alpha=0.5, ExternalForce="0,0,0") {
    angularDampingOverride=1
    angularSpringConstant=5000
    bAngularSpring=true
    bOverrideAngularDamping=true
    bUseGravityOverride=true
    gravityScale=0.35
    simulationSpace=RootRelative
    alphaBoolBlend {blendOption=Linear}
    boundBone {boneName=dyn_cloth_c_02_r}
    gravityOverride {z=-980}
    physicsBodyDefinitions [{
      collisionType=OuterSphere
      boundBone {boneName=dyn_cloth_c_02_r}
      boxExtents=<15 25 25>
      constraintSetup {angularLimitsMax=<0 3 6>, angularLimitsMin=<0 -3 -6>, angularTarget=<1 0 0>}
      localJointOffset=<2 0 0>
     }]
   }
   n14 = Anim Dynamics - Bone: dyn_cloth_01_b(ComponentPose=@n13, Alpha=1, ExternalForce="0,0,0") {
    angularDampingOverride=1
    angularSpringConstant=5000
    bAngularSpring=true
    bOverrideAngularDamping=true
    bUseGravityOverride=true
    gravityScale=0.35
    simulationSpace=RootRelative
    alphaBoolBlend {blendOption=Linear}
    boundBone {boneName=dyn_cloth_01_b}
    gravityOverride {z=-980}
    physicsBodyDefinitions [{
      collisionType=OuterSphere
      boundBone {boneName=dyn_cloth_01_b}
      boxExtents=<15 25 25>
      constraintSetup {angularLimitsMax=<0 5 6>, angularLimitsMin=<0 -5 -6>, angularTarget=<1 0 0>}
      localJointOffset=<2 0 0>
     }]
   }
   n15 = Anim Dynamics - Bone: dyn_cloth_02_b(ComponentPose=@n14, Alpha=1, ExternalForce="0,0,0") {
    angularDampingOverride=1
    angularSpringConstant=5000
    bAngularSpring=true
    bOverrideAngularDamping=true
    bUseGravityOverride=true
    gravityScale=0.35
    simulationSpace=RootRelative
    alphaBoolBlend {blendOption=Linear}
    boundBone {boneName=dyn_cloth_02_b}
    gravityOverride {z=-980}
    physicsBodyDefinitions [{
      collisionType=OuterSphere
      boundBone {boneName=dyn_cloth_02_b}
      boxExtents=<15 25 25>
      constraintSetup {angularLimitsMax=<0 5 8>, angularLimitsMin=<0 -5 -8>, angularTarget=<1 0 0>}
      localJointOffset=<2 0 0>
     }]
   }
   n16 = Anim Dynamics - Bone: dyn_cloth_03_b(ComponentPose=@n15, Alpha=1, ExternalForce="0,0,0") {
    angularDampingOverride=1
    angularSpringConstant=5000
    bAngularSpring=true
    bOverrideAngularDamping=true
    bUseGravityOverride=true
    gravityScale=0.35
    simulationSpace=RootRelative
    alphaBoolBlend {blendOption=Linear}
    boundBone {boneName=dyn_cloth_03_b}
    gravityOverride {z=-980}
    physicsBodyDefinitions [{
      collisionType=OuterSphere
      boundBone {boneName=dyn_cloth_03_b}
      boxExtents=<15 25 25>
      constraintSetup {angularLimitsMax=<0 5 12>, angularLimitsMin=<0 -5 -12>, angularTarget=<1 0 0>}
      localJointOffset=<2 0 0>
     }]
   }
   n17 = Anim Dynamics - Bone: dyn_cloth_01_f(ComponentPose=@n16, Alpha=1, ExternalForce="0,0,0") {
    angularDampingOverride=1
    angularSpringConstant=5000
    bAngularSpring=true
    bOverrideAngularDamping=true
    bUseGravityOverride=true
    gravityScale=0.35
    simulationSpace=RootRelative
    alphaBoolBlend {blendOption=Linear}
    boundBone {boneName=dyn_cloth_01_f}
    gravityOverride {z=-980}
    physicsBodyDefinitions [{
      collisionType=OuterSphere
      boundBone {boneName=dyn_cloth_01_f}
      boxExtents=<15 25 25>
      constraintSetup {angularLimitsMax=<0 5 12>, angularLimitsMin=<0 -5 -12>, angularTarget=<1 0 0>}
      localJointOffset=<2 0 0>
     }]
   }
   n18 = Get bIsSkyDiving()
   n19 = "Transform (Modify) Bone - Bone: dyn_cloth_01_b"!(ComponentPose=@n17, bAlphaBoolEnabled:bool=@n18, Translation="0,0,-2", Rotation="8,-5,0", Scale="1,1,1") -> (Pose) {
    alphaInputType=Bool
    rotationMode=BMM_Additive
    rotationSpace=BCS_BoneSpace
    translationMode=BMM_Additive
    translationSpace=BCS_BoneSpace
    alphaBoolBlend {blendOption=Linear}
    boneToModify {boneName=dyn_cloth_01_b}
    rotation {pitch=8, yaw=-5}
    translation {z=-2}
   }
   n20 = "Transform (Modify) Bone - Bone: dyn_cloth_02_b"!(ComponentPose=@n19, bAlphaBoolEnabled:bool=@n18, Translation="0,0,-2", Rotation="8,-5,0", Scale="1,1,1") -> (Pose) {
    alphaInputType=Bool
    rotationMode=BMM_Additive
    rotationSpace=BCS_BoneSpace
    translationMode=BMM_Additive
    translationSpace=BCS_BoneSpace
    alphaBoolBlend {blendOption=Linear}
    boneToModify {boneName=dyn_cloth_02_b}
    rotation {pitch=8, yaw=-5}
    translation {z=-2}
   }
   n21 = "Transform (Modify) Bone - Bone: dyn_cloth_c_01_l"!(ComponentPose=@n20, bAlphaBoolEnabled:bool=@n18, Translation="0,0,-2", Rotation="12,-12,0", Scale="1,1,1") -> (Pose) {
    alphaInputType=Bool
    rotationMode=BMM_Additive
    rotationSpace=BCS_BoneSpace
    translationMode=BMM_Additive
    translationSpace=BCS_BoneSpace
    alphaBoolBlend {blendOption=Linear}
    boneToModify {boneName=dyn_cloth_c_01_l}
    rotation {pitch=12, yaw=-12}
    translation {z=-2}
   }
   n22 = "Transform (Modify) Bone - Bone: dyn_cloth_c_02_l"!(ComponentPose=@n21, bAlphaBoolEnabled:bool=@n18, Translation="0,0,-2", Rotation="12,-12,0", Scale="1,1,1") -> (Pose) {
    alphaInputType=Bool
    rotationMode=BMM_Additive
    rotationSpace=BCS_BoneSpace
    translationMode=BMM_Additive
    translationSpace=BCS_BoneSpace
    alphaBoolBlend {blendOption=Linear}
    boneToModify {boneName=dyn_cloth_c_02_l}
    rotation {pitch=12, yaw=-12}
    translation {z=-2}
   }
   n23 = "Transform (Modify) Bone - Bone: dyn_cloth_c_01_r"!(ComponentPose=@n22, bAlphaBoolEnabled:bool=@n18, Translation="0,0,-2", Rotation="12,-12,0", Scale="1,1,1") -> (Pose) {
    alphaInputType=Bool
    rotationMode=BMM_Additive
    rotationSpace=BCS_BoneSpace
    translationMode=BMM_Additive
    translationSpace=BCS_BoneSpace
    alphaBoolBlend {blendOption=Linear}
    boneToModify {boneName=dyn_cloth_c_01_r}
    rotation {pitch=12, yaw=-12}
    translation {z=-2}
   }
   n24 = "Transform (Modify) Bone - Bone: dyn_cloth_c_02_r"!(ComponentPose=@n23, bAlphaBoolEnabled:bool=@n18, Translation="0,0,-2", Rotation="12,-12,0", Scale="1,1,1") -> (Pose) {
    alphaInputType=Bool
    rotationMode=BMM_Additive
    rotationSpace=BCS_BoneSpace
    translationMode=BMM_Additive
    translationSpace=BCS_BoneSpace
    alphaBoolBlend {blendOption=Linear}
    boneToModify {boneName=dyn_cloth_c_02_r}
    rotation {pitch=12, yaw=-12}
    translation {z=-2}
   }
   n25 = Get bIsCrouching()
   n26 = "Transform (Modify) Bone - Bone: dyn_cloth_c_01_l"(ComponentPose=@n24, bAlphaBoolEnabled=@n25, Translation="0,0,0", Rotation="0,-7,0", Scale="1,1,1") {
    alphaInputType=Bool
    rotationMode=BMM_Additive
    rotationSpace=BCS_BoneSpace
    translationMode=BMM_Additive
    translationSpace=BCS_BoneSpace
    alphaBoolBlend {blendOption=Linear}
    boneToModify {boneName=dyn_cloth_c_01_l}
    rotation {yaw=-7}
   }
   n27 = "Transform (Modify) Bone - Bone: dyn_cloth_c_01_r"(ComponentPose=@n26, bAlphaBoolEnabled=@n25, Translation="0,0,0", Rotation="0,-7,0", Scale="1,1,1") {
    alphaInputType=Bool
    rotationMode=BMM_Additive
    rotationSpace=BCS_BoneSpace
    translationMode=BMM_Additive
    translationSpace=BCS_BoneSpace
    alphaBoolBlend {blendOption=Linear}
    boneToModify {boneName=dyn_cloth_c_01_r}
    rotation {yaw=-7}
   }
   Output Pose(Result=Component To Local(ComponentPose=@n27)) {layerGroup=None}
  ]}
 EventGraph {nodes [Event BlueprintUpdateAnimation()]}
]
variables [Gravity Override {type=Vector}]
