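# Physics asset for the AgedChasm female medium face accessory mesh, setting up left and right cloth lace chains for physics.
# Each chain has a small base capsule and three dyn capsules, linked in series with limited swing and zero twist limit.
# Both bases are constrained to spine_04 and spine_05 with free rotation.
# Large capsules on spine_03 through spine_05 form upper-torso bodies, likewise linked with free angular motion.
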PhysicsAsset /BRCosmetics/Characters/Player/Female/Medium/Bodies/F_MED_AgedChasm/Meshes/Parts/F_MED_AgedChasm_FaceAcc_Physics
bodies {
 base_cloth_laces_l [{shapeType=Capsule, length=0.505, radius=0.505, shapeName=base_cloth_laces_l_capsule, rotation=<rot 0 -90 0>}]
 base_cloth_laces_r [{shapeType=Capsule, length=0.505, radius=0.505, shapeName=base_cloth_laces_r_capsule, rotation=<rot 0 -90 0>}]
 dyn_cloth_laces_01_l [{
   shapeType=Capsule
   length=1.9404
   radius=0.5455
   shapeName=dyn_cloth_laces_01_r_capsule
   center=<2.2995 0 -0.1236>
   rotation=<rot 0 -85.1639 -87.5658>
  }]
 dyn_cloth_laces_01_r [{
   shapeType=Capsule
   length=1.9404
   radius=0.5
   shapeName=dyn_cloth_laces_01_r_capsule
   center=<2.2995 0.0452 -0.1236>
   rotation=<rot 0 -85.1639 -87.5658>
  }]
 dyn_cloth_laces_02_l [{
   shapeType=Capsule
   length=0.7035
   radius=0.55
   shapeName=dyn_cloth_laces_02_r_capsule
   center=<2.5138 0.14 -0.2001>
   rotation=<rot 0 -93.4365 -87.9797>
  }]
 dyn_cloth_laces_02_r [{
   shapeType=Capsule
   length=0.7035
   radius=0.54
   shapeName=dyn_cloth_laces_02_r_capsule
   center=<2.5138 -0.2 -0.2001>
   rotation=<rot 0 -93.4365 -87.9797>
  }]
 dyn_cloth_laces_03_l [{
   shapeType=Capsule
   length=1.8429
   radius=0.61
   shapeName=dyn_cloth_laces_03_r_capsule
   center=<3.0446 -0.3389 -0.1212>
   rotation=<rot 0 83.0658 -87.205>
  }]
 dyn_cloth_laces_03_r [{
   shapeType=Capsule
   length=1.8429
   radius=0.6
   shapeName=dyn_cloth_laces_03_r_capsule
   center=<2.7248 -0.6 -0.1369>
   rotation=<rot 0 83.0658 -87.205>
  }]
 spine_03 [{shapeType=Capsule, length=12, radius=6.5, shapeName=spine_04_capsule, center=<9.4344 5.8123 0>, rotation=<rot 0 179.9998 -180>}]
 spine_04 [{shapeType=Capsule, length=11.5, radius=6.5, shapeName=spine_04_capsule, center=<7.4978 2.4424 0>, rotation=<rot 0 179.9998 -180>}]
 spine_05 [{shapeType=Capsule, length=19.1148, radius=9, center=<0.5319 -0.8835 0>}]
}
constraints [
 {
  bone1Name=base_cloth_laces_l
  bone2Name=spine_04
  swing1LimitDegrees=45
  swing1Motion=Free
  swing2LimitDegrees=45
  swing2Motion=Free
  twistLimitDegrees=45
  twistMotion=Free
 }
 {bone1Name=dyn_cloth_laces_01_l, bone2Name=base_cloth_laces_l, swing1LimitDegrees=30, swing2LimitDegrees=30, twistLimitDegrees=0}
 {bone1Name=dyn_cloth_laces_02_l, bone2Name=dyn_cloth_laces_01_l, swing1LimitDegrees=30, swing2LimitDegrees=30, twistLimitDegrees=0}
 {bone1Name=dyn_cloth_laces_03_l, bone2Name=dyn_cloth_laces_02_l, swing1LimitDegrees=30, swing2LimitDegrees=30, twistLimitDegrees=0}
 {
  bone1Name=base_cloth_laces_r
  bone2Name=spine_04
  swing1LimitDegrees=45
  swing1Motion=Free
  swing2LimitDegrees=45
  swing2Motion=Free
  twistLimitDegrees=45
  twistMotion=Free
 }
 {bone1Name=dyn_cloth_laces_01_r, bone2Name=base_cloth_laces_r, swing1LimitDegrees=30, swing2LimitDegrees=30, twistLimitDegrees=0}
 {bone1Name=dyn_cloth_laces_02_r, bone2Name=dyn_cloth_laces_01_r, swing1LimitDegrees=30, swing2LimitDegrees=30, twistLimitDegrees=0}
 {bone1Name=dyn_cloth_laces_03_r, bone2Name=dyn_cloth_laces_02_r, swing1LimitDegrees=30, swing2LimitDegrees=30, twistLimitDegrees=0}
 {
  bone1Name=spine_04
  bone2Name=spine_03
  swing1LimitDegrees=45
  swing1Motion=Free
  swing2LimitDegrees=45
  swing2Motion=Free
  twistLimitDegrees=45
  twistMotion=Free
 }
 {
  bone1Name=spine_05
  bone2Name=spine_04
  swing1LimitDegrees=45
  swing1Motion=Free
  swing2LimitDegrees=45
  swing2Motion=Free
  twistLimitDegrees=45
  twistMotion=Free
 }
 {
  bone1Name=base_cloth_laces_l
  bone2Name=spine_05
  swing1LimitDegrees=45
  swing1Motion=Free
  swing2LimitDegrees=45
  swing2Motion=Free
  twistLimitDegrees=45
  twistMotion=Free
 }
 {
  bone1Name=base_cloth_laces_r
  bone2Name=spine_05
  swing1LimitDegrees=45
  swing1Motion=Free
  swing2LimitDegrees=45
  swing2Motion=Free
  twistLimitDegrees=45
  twistMotion=Free
 }
]
